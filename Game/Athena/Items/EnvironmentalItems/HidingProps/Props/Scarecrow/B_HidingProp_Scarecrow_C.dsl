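; Scarecrow hiding prop that can turn to follow the view of the player hiding inside it.
; When the hiding player count changes, the server starts or stops a looping timer, with rotation gated by a scalable-float flag.
; Each timer step eases actor rotation toward the first hider's control yaw, offset by minus 90 degrees.
; CurrentRotation's OnRep applies that rotation on clients.
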(event EventBeginPlay
  (|CallParentFunction) ; This node is disabled and will not be called.\nDrag off pins to build functionality.
)

(event Collision|EventActorBeginOverlap
  (params (OtherActor :type Actor%20Object%20Reference))
  (|CallParentFunction OtherActor) ; This node is disabled and will not be called.\nDrag off pins to build functionality.
)

(event EventTick
  (params (DeltaSeconds :type Float%20%28single-precision%29))
  (|CallParentFunction DeltaSeconds) ; This node is disabled and will not be called.\nDrag off pins to build functionality.
)

(event EventHidingPlayerCountChanged
  (bind _hidingplayers (Variables|Default|GetHidingPlayers))
  (|CallParentFunction)
  (Networking|SwitchHasAuthority
    (:Authority
      (if (> (Utilities|Array|Length _hidingplayers) 0)
        (Variables|Default|SetIsRotatingWithPlayer (ScalableFloat|GetValueasBool (Variables|Default|GetRotateWithPlayerBool)))
        (Variables|Default|SetPlayerReference (Utilities|Array|Get(acopy) _hidingplayers))
        (Variables|Default|SetCurrentRotation (Transformation|GetActorRotation))
        (bind _returnvalue (Utilities|Time|SetTimerbyEvent (delegate Custom|MoveWithCameraLogic) (ScalableFloat|GetValueatLevel (Variables|Default|GetTick%20Value)) true true))
        (Variables|Default|SetUpdateCameraTimer _returnvalue)
        (else
          (CallFunction|StopMoveCameraLogic))))))

(event Custom|MoveWithCameraLogic
  (bind _interp_speed (Variables|Default|GetInterp%20Speed))
  (bind _returnvalue (ScalableFloat|GetValueatLevel _interp_speed))
  (bind _returnvalue_2 (Utilities|Time|GetWorldDeltaSeconds))
  (bind _targetrotation (Variables|Default|GetTargetRotation))
  (bind _currentrotation (Variables|Default|GetCurrentRotation))
  (bind _returnvalue_3 (Math|Interpolation|RInterpTo _currentrotation _targetrotation _returnvalue_2 _returnvalue))
  (if (Variables|Default|GetIsRotatingWithPlayer)
    (Utilities|IsValid (Variables|Default|GetPlayerReference)
      (:Is%20Valid
        (Variables|Default|SetTargetRotation (Math|Rotator|MakeRotator 0.0 0.0 (+ (.yaw (Pawn|GetControlRotation (Variables|Default|GetPlayerReference) :!class /Script/Engine.Pawn)) -9e+01)))
        (Transformation|SetActorRotation _returnvalue_3)
        (Variables|Default|SetCurrentRotation _returnvalue_3))
      (:Is%20Not%20Valid
        (CallFunction|StopMoveCameraLogic)))
    (else
      (CallFunction|StopMoveCameraLogic))))

(event Custom|StopMoveCameraLogic
  (Variables|Default|SetIsRotatingWithPlayer false)
  (Variables|Default|SetPlayerReference 0)
  (Utilities|Time|ClearandInvalidateTimerbyHandle (Variables|Default|GetUpdateCameraTimer)))

(fn OnRep_CurrentRotation
  (Transformation|SetActorRotation (Variables|Default|GetCurrentRotation)))

(fn ConstructionScript
  (|CallParentFunction))
